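# Animation Blueprint for the FruitFire Nudge backpack cosmetic, adding secondary physics motion to its ball bones.
# Three AnimDynamics nodes run in series on the input pose: a linear spring on ball_trans, then two angular-spring passes on ball_rot, each limited to a different axis.
# Component motion and gravity drive the simulation.
# The event graph lowers SkyDiveAlpha while skydiving, but the AnimGraph never reads it, as every node's alpha stays at 1.
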AnimBlueprint /BRCosmetics/Accessories/FORT_Backpacks/Backpack_F_MED_FruitFire_Nudge/Meshes/F_MED_FruitFire_Nudge_Pack_AnimBP
target_skeleton=&/F_MED_FruitFire_Nudge_Pack_Skeleton
graphs [
 AnimGraph {nodes [
   n1 = Anim Dynamics - Bone: ball_trans(ComponentPose=Local To Component(LocalPose=Input Pose()), Alpha=1, ExternalForce="0,0,0") {
    angularBiasOverride=0.999
    bLinearSpring=true
    bOverrideLinearDamping=true
    bUseGravityOverride=true
    linearDampingOverride=1
    linearSpringConstant=50
    boundBone {boneName=ball_trans}
    componentAppliedLinearAccClamp=<100 100 1100>
    componentLinearAccScale=<0.1 0.1 0.3>
    componentLinearVelScale=<0.2 0.2 0.37>
    gravityOverride {z=-980}
    physicsBodyDefinitions [{
      boundBone {boneName=ball_trans}
      boxExtents=<30 30 30>
      constraintSetup {angularLimitsMax=<0 0 0>, angularLimitsMin=<0 0 0>, angularTarget=<1 0 0>, linearAxesMin=<0 -4 0>}
      localJointOffset=<0 0 0>
     }]
   }
   n2 = Anim Dynamics - Bone: ball_rot(ComponentPose=@n1, Alpha=1, ExternalForce="0,0,0") {
    angularDampingOverride=1
    angularSpringConstant=5000
    bAngularSpring=true
    bOverrideAngularDamping=true
    bUseGravityOverride=true
    boundBone {boneName=ball_rot}
    componentAppliedLinearAccClamp=<1000 1000 1000>
    componentLinearAccScale=<0.5 0.5 0.5>
    componentLinearVelScale=<0.7 0.7 0.7>
    gravityOverride {z=-980}
    physicsBodyDefinitions [{
      boundBone {boneName=ball_rot}
      boxExtents=<42 36 36>
      constraintSetup {angularLimitsMax=<0 0 5>, angularLimitsMin=<0 0 -20>, angularTarget=<1 0 0>}
      localJointOffset=<0.5 0 0>
     }]
   }
   n3 = Anim Dynamics - Bone: ball_rot(ComponentPose=@n2, Alpha=1, ExternalForce="0,0,0") {
    angularDampingOverride=0.999
    angularSpringConstant=2000
    bAngularSpring=true
    bOverrideAngularDamping=true
    bUseGravityOverride=true
    boundBone {boneName=ball_rot}
    componentAppliedLinearAccClamp=<10000 10000 10000>
    componentLinearAccScale=<0.5 0.5 0.5>
    componentLinearVelScale=<0.7 0.7 0.7>
    gravityOverride {z=-980}
    physicsBodyDefinitions [{
      boundBone {boneName=ball_rot}
      boxExtents=<30 20 25>
      constraintSetup {angularLimitsMax=<0 20 0>, angularLimitsMin=<0 -20 0>, angularTarget=<1 0 0>}
      localJointOffset=<1 0 0>
     }]
   }
   Output Pose(Result=Component To Local(ComponentPose=@n3)) {layerGroup=None}
  ]}
 EventGraph {nodes [
   Event BlueprintUpdateAnimation()
   > Set SkyDiveAlpha(SkyDiveAlpha:double=SelectFloat(A=0.3, B=1, bPickA=Get bIsSkyDiving())) -> (Output_Get:double)
   TryGetPawnOwner()
  ]}
]
variables [SkyDiveAlpha {type=number}]
